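# Physics asset for secondary motion on the M_MED_Wombat male body's shirt, collars, sleeve, straps, knee strings and belt knives.
# Capsules on pelvis, spine, arms and legs anchor short dyn_ chains of boxes and capsules, mostly tightly angle-limited.
# Back shirt tails get a wide swing2 and a free constraint linking their tips.
# Core joints are locked except clavicles and upperarm_r, and all dyn_r_sleeve bodies attach to upperarm_l despite their name.
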PhysicsAsset /BRCosmetics/Characters/Player/Male/Medium/Bodies/M_MED_Wombat/Meshes/M_MED_Wombat_Physics_shirt
bodies {
 calf_l [{shapeType=Capsule, length=26.9939, radius=6.459, center=<-19.7691 -2.1104 0>, rotation=<rot 90 0 0>}]
 calf_r [{shapeType=Capsule, length=34.5544, radius=7.482, center=<18.3583 0 0>, rotation=<rot 90 0 0>}]
 clavicle_l [{shapeType=Capsule, length=9.727, radius=7.0034, center=<9.3098 0.0002 0>, rotation=<rot 90 0 0>}]
 clavicle_r [{shapeType=Capsule, length=10.0286, radius=7.4437, center=<-8.5926 0.0002 0>, rotation=<rot 90 0 0>}]
 dyn_back_strap_01 [{shapeType=Capsule, length=2, radius=2, shapeName=dyn_back_strap_01_capsule, center=<2.773 0 0>, rotation=<rot 0 0 90>}]
 dyn_back_strap_02 [{shapeType=Capsule, length=1.3326, radius=2.6674, shapeName=dyn_back_strap_02_capsule, center=<2.773 0 0>, rotation=<rot 0 0 90>}]
 dyn_back_strap_03 [{shapeType=Capsule, length=2, radius=2, shapeName=dyn_back_strap_03_capsule, center=<2.773 0 0>, rotation=<rot 0 0 90>}]
 dyn_belt_knife_l_01 [{shapeType=Box, extentX=4.7869, extentY=4.9886, extentZ=2.3651, center=<2.4462 -0.0003 0>}]
 dyn_belt_knife_l_02 [{shapeType=Box, extentX=9.0497, extentY=5.429, extentZ=2, center=<4.745 -0.0001 0>}]
 dyn_belt_knife_r_01 [{shapeType=Box, extentX=4.7869, extentY=4.9886, extentZ=2.3651, center=<2.4462 -0.0003 0>}]
 dyn_belt_knife_r_02 [{shapeType=Box, extentX=9.0497, extentY=5.429, extentZ=2, center=<4.745 -0.0001 0>}]
 dyn_l_collar_01 [{shapeType=Box, extentX=2, extentY=2, extentZ=2, center=<1.8364 0 0>}]
 dyn_l_collar_02 [{shapeType=Box, extentX=3.7163, extentY=2, extentZ=0.8942, center=<2.1992 0.0001 0>}]
 dyn_l_knee_loop_l_01 [{shapeType=Box, extentX=2.5422, extentY=2, extentZ=0.8658, center=<1.3689 0 -0.0001>}]
 dyn_l_knee_loop_l_02 [{shapeType=Box, extentX=2.6985, extentY=2.6979, extentZ=0.8658, center=<1.5023 0 0.0001>}]
 dyn_l_knee_string_long_01 [{shapeType=Box, extentX=5.5215, extentY=0.973, extentZ=1.0155, center=<2.8023 0 0>}]
 dyn_l_knee_string_long_02 [{shapeType=Box, extentX=5.697, extentY=0.973, extentZ=1.0155, center=<2.8441 0 0>}]
 dyn_l_knee_string_long_03 [{shapeType=Box, extentX=4.0282, extentY=1.3138, extentZ=1.0155, center=<2.2178 0 0.0001>}]
 dyn_l_knee_string_r_loop_01 [{shapeType=Box, extentX=2, extentY=2, extentZ=0.8658, center=<0.9077 0 0>}]
 dyn_l_knee_string_r_loop_02 [{shapeType=Box, extentX=2, extentY=2, extentZ=0.8658, center=<1.2488 0 0>}]
 dyn_l_knee_string_r_loop_03 [{shapeType=Box, extentX=4.3336, extentY=2.9102, extentZ=0.8658, center=<2.2708 0 0.0001>}]
 dyn_lb_collar_01 [{shapeType=Box, extentX=2, extentY=2, extentZ=1.1488, center=<1.1881 0 0>}]
 dyn_lb_collar_02 [{shapeType=Box, extentX=3.7163, extentY=2, extentZ=0.8942, center=<2.1992 0.0001 0>}]
 dyn_r_collar_01 [{shapeType=Box, extentX=2, extentY=2, extentZ=2, center=<1.1881 0 0>}]
 dyn_r_collar_02 [{shapeType=Box, extentX=3.7163, extentY=2, extentZ=0.8942, center=<2.1992 0.0001 0>}]
 dyn_r_knee_string_01 [{shapeType=Box, extentX=3.1986, extentY=1.4348, extentZ=0.9478, center=<1.7228 0 0.0001>}]
 dyn_r_knee_string_02 [{shapeType=Box, extentX=4.6133, extentY=1.4401, extentZ=0.9866, center=<2.3797 0 0.0001>}]
 dyn_r_knee_string_03 [{shapeType=Box, extentX=4.0551, extentY=1.3888, extentZ=0.9866, center=<2.0548 0 0>}]
 dyn_r_knee_string_r_01 [{shapeType=Box, extentX=3.4314, extentY=3.1508, extentZ=0.7249, center=<1.794 0 0.0001>}]
 dyn_r_sleeve_01 [{shapeType=Capsule, length=3.899, radius=1.0763, center=<2.7754 -0.0003 0.6561>, rotation=<rot 90 0 -90>}]
 dyn_r_sleeve_02 [{shapeType=Capsule, length=3.0544, radius=1.9208, center=<2.7754 -0.0002 0.1744>, rotation=<rot 90 26.565 -63.4349>}]
 dyn_r_sleeve_03 [{shapeType=Capsule, length=3.5477, radius=1.4276, center=<2.7754 -0.0002 -0.5532>, rotation=<rot 90 0 -90>}]
 dyn_r_sleeve_04 [{shapeType=Capsule, length=3.3866, radius=1.5886, center=<2.7754 -0.0002 -0.4275>, rotation=<rot 90 -4.7636 -94.7636>}]
 dyn_r_sleeve_05 [{shapeType=Capsule, length=3.4432, radius=1.532, center=<2.7754 -0.0002 0>, rotation=<rot 90 90 0>}]
 dyn_r_sleeve_06 [{shapeType=Capsule, length=3.899, radius=1.0763, center=<2.7754 -0.0002 0>, rotation=<rot 90 90 0>}]
 dyn_r_string_r_01 [{shapeType=Box, extentX=2, extentY=1.0488, extentZ=0.9119, center=<0.9135 0 0.0001>}]
 dyn_r_string_r_02 [{shapeType=Box, extentX=2.9017, extentY=1.1991, extentZ=0.6983, center=<1.4766 0 0>}]
 dyn_rb_collar_01 [{shapeType=Box, extentX=2, extentY=2, extentZ=2, center=<1.1881 0 0>}]
 dyn_rb_collar_02 [{shapeType=Box, extentX=3.7163, extentY=2, extentZ=0.8942, center=<2.1992 0.0001 0>}]
 dyn_shirt_back_l_01 [{shapeType=Box, extentX=4.1297, extentY=9.5582, extentZ=1.8033, center=<2.1708 0 0.0001>}]
 dyn_shirt_back_l_02 [{shapeType=Capsule, length=7.7278, radius=2.3127, shapeName=dyn_shirt_back_l_02_capsule, center=<2.5531 0 0>, rotation=<rot 0 0 90>}]
 dyn_shirt_back_r_01 [{shapeType=Box, extentX=4.1297, extentY=9.5582, extentZ=1.8033, center=<2.1708 0 0.0001>}]
 dyn_shirt_back_r_02 [{shapeType=Capsule, length=7.7278, radius=2.3127, shapeName=dyn_shirt_back_r_02_capsule, center=<2.5531 0 0>, rotation=<rot 0 0 90>}]
 dyn_shirt_front_l_01 [{shapeType=Box, extentX=4.1297, extentY=9.5582, extentZ=5.4902, center=<2.1708 0 0.0001>}]
 dyn_shirt_front_l_02 [{shapeType=Box, extentX=2.6705, extentY=9.5582, extentZ=5.4902, center=<1.4739 0 0>}]
 dyn_shirt_front_r_01 [{shapeType=Box, extentX=4.1297, extentY=9.5582, extentZ=4.4787, center=<2.1708 0 0.0001>}]
 dyn_shirt_front_r_02 [{shapeType=Box, extentX=2.6705, extentY=9.5582, extentZ=4.4787, center=<1.4739 0 0>}]
 pelvis [
  {shapeType=Capsule, length=13.4921, radius=10.4527, center=<-2.6383 -1.0313 -0.0001>}
  {shapeType=Capsule, length=10.2417, radius=11.3658, center=<-9.7316 -0.0003 0>}
 ]
 spine_05 [{shapeType=Capsule, length=15.3054, radius=10.8425}]
 thigh_l [{shapeType=Capsule, length=33.4511, radius=7.459, center=<-21.0021 -0.0009 0>, rotation=<rot 90 0 0>}]
 thigh_r [{shapeType=Capsule, length=26.4639, radius=7.459, center=<20.4554 -0.0001 -0.0003>, rotation=<rot 90 0 0>}]
 upperarm_l [{shapeType=Capsule, length=15.296, radius=4.581, center=<13.4971 0.0009 -1.0095>, rotation=<rot 90 0 0>}]
 upperarm_r [{shapeType=Capsule, length=18.2553, radius=5.9888, center=<-15.9751 0.0005 0>, rotation=<rot 90 0 0>}]
}
constraints [
 {
  bone1Name=upperarm_l
  bone2Name=clavicle_l
  swing1LimitDegrees=45
  swing1Motion=Locked
  swing2LimitDegrees=45
  swing2Motion=Locked
  twistLimitDegrees=45
  twistMotion=Locked
 }
 {bone1Name=dyn_r_sleeve_01, bone2Name=upperarm_l, swing1LimitDegrees=4.5, swing2LimitDegrees=8, twistLimitDegrees=3}
 {bone1Name=dyn_r_sleeve_03, bone2Name=upperarm_l, swing1LimitDegrees=4.5, swing2LimitDegrees=8, twistLimitDegrees=3}
 {bone1Name=dyn_r_sleeve_04, bone2Name=upperarm_l, swing1LimitDegrees=4.5, swing2LimitDegrees=8, twistLimitDegrees=3}
 {bone1Name=dyn_r_sleeve_05, bone2Name=upperarm_l, swing1LimitDegrees=4.5, swing2LimitDegrees=8, twistLimitDegrees=3}
 {bone1Name=dyn_r_sleeve_02, bone2Name=upperarm_l, swing1LimitDegrees=4.5, swing2LimitDegrees=8, twistLimitDegrees=3}
 {bone1Name=dyn_r_sleeve_06, bone2Name=upperarm_l, swing1LimitDegrees=4.5, swing2LimitDegrees=8, twistLimitDegrees=3}
 {bone1Name=dyn_l_collar_02, bone2Name=dyn_l_collar_01, swing1LimitDegrees=5, swing2LimitDegrees=10, twistLimitDegrees=3}
 {bone1Name=dyn_l_collar_01, bone2Name=spine_05, swing1LimitDegrees=3, swing2LimitDegrees=5, twistLimitDegrees=2}
 {bone1Name=dyn_r_collar_02, bone2Name=dyn_r_collar_01, swing1LimitDegrees=5, swing2LimitDegrees=10, twistLimitDegrees=3}
 {bone1Name=dyn_r_collar_01, bone2Name=spine_05, swing1LimitDegrees=3, swing2LimitDegrees=5, twistLimitDegrees=2}
 {bone1Name=dyn_rb_collar_01, bone2Name=spine_05, swing1LimitDegrees=3, swing2LimitDegrees=5, twistLimitDegrees=2}
 {bone1Name=dyn_rb_collar_02, bone2Name=dyn_rb_collar_01, swing1LimitDegrees=5, swing2LimitDegrees=10, twistLimitDegrees=3}
 {bone1Name=dyn_lb_collar_02, bone2Name=dyn_lb_collar_01, swing1LimitDegrees=5, swing2LimitDegrees=10, twistLimitDegrees=3}
 {bone1Name=dyn_lb_collar_01, bone2Name=spine_05, swing1LimitDegrees=3, swing2LimitDegrees=5, twistLimitDegrees=2}
 {
  bone1Name=calf_l
  bone2Name=thigh_l
  swing1LimitDegrees=45
  swing1Motion=Locked
  swing2LimitDegrees=45
  swing2Motion=Locked
  twistLimitDegrees=45
  twistMotion=Locked
 }
 {bone1Name=dyn_l_knee_loop_l_01, bone2Name=calf_l, swing1LimitDegrees=10, swing2LimitDegrees=20, twistLimitDegrees=7}
 {bone1Name=dyn_l_knee_loop_l_02, bone2Name=dyn_l_knee_loop_l_01, swing1LimitDegrees=10, swing2LimitDegrees=20, twistLimitDegrees=7}
 {bone1Name=dyn_l_knee_string_r_loop_01, bone2Name=calf_l, swing1LimitDegrees=10, swing2LimitDegrees=10, twistLimitDegrees=5}
 {
  bone1Name=dyn_l_knee_string_r_loop_02
  bone2Name=dyn_l_knee_string_r_loop_01
  swing1LimitDegrees=10
  swing2LimitDegrees=10
  twistLimitDegrees=5
 }
 {
  bone1Name=dyn_l_knee_string_r_loop_03
  bone2Name=dyn_l_knee_string_r_loop_02
  swing1LimitDegrees=10
  swing2LimitDegrees=10
  twistLimitDegrees=5
 }
 {
  bone1Name=calf_r
  bone2Name=thigh_r
  swing1LimitDegrees=45
  swing1Motion=Locked
  swing2LimitDegrees=45
  swing2Motion=Locked
  twistLimitDegrees=45
  twistMotion=Locked
 }
 {
  bone1Name=thigh_r
  bone2Name=pelvis
  swing1LimitDegrees=45
  swing1Motion=Locked
  swing2LimitDegrees=45
  swing2Motion=Locked
  twistLimitDegrees=45
  twistMotion=Locked
 }
 {bone1Name=dyn_r_knee_string_01, bone2Name=thigh_r, swing1LimitDegrees=7, swing2LimitDegrees=12, twistLimitDegrees=5}
 {bone1Name=dyn_r_knee_string_02, bone2Name=dyn_r_knee_string_01, swing1LimitDegrees=7, swing2LimitDegrees=12, twistLimitDegrees=5}
 {bone1Name=dyn_r_knee_string_03, bone2Name=dyn_r_knee_string_02, swing1LimitDegrees=7, swing2LimitDegrees=12, twistLimitDegrees=5}
 {bone1Name=dyn_r_knee_string_r_01, bone2Name=thigh_r, swing1LimitDegrees=10, swing2LimitDegrees=10, twistLimitDegrees=5}
 {bone1Name=dyn_r_string_r_01, bone2Name=thigh_r, swing1LimitDegrees=5, swing2LimitDegrees=10, twistLimitDegrees=5}
 {bone1Name=dyn_r_string_r_02, bone2Name=dyn_r_string_r_01, swing1LimitDegrees=5, swing2LimitDegrees=10, twistLimitDegrees=5}
 {bone1Name=dyn_belt_knife_r_01, bone2Name=pelvis, swing1LimitDegrees=7, swing2LimitDegrees=15, twistLimitDegrees=3}
 {bone1Name=dyn_belt_knife_r_02, bone2Name=dyn_belt_knife_r_01, swing1LimitDegrees=7, swing2LimitDegrees=10, twistLimitDegrees=3}
 {bone1Name=dyn_belt_knife_l_01, bone2Name=pelvis, swing1LimitDegrees=7, swing2LimitDegrees=15, twistLimitDegrees=3}
 {bone1Name=dyn_belt_knife_l_02, bone2Name=dyn_belt_knife_l_01, swing1LimitDegrees=7, swing2LimitDegrees=10, twistLimitDegrees=3}
 {bone1Name=dyn_shirt_front_l_01, bone2Name=pelvis, swing1LimitDegrees=3, swing2LimitDegrees=10, twistLimitDegrees=3}
 {bone1Name=dyn_shirt_front_r_01, bone2Name=pelvis, swing1LimitDegrees=3, swing2LimitDegrees=10, twistLimitDegrees=3}
 {bone1Name=dyn_shirt_back_r_01, bone2Name=pelvis, swing1LimitDegrees=5, swing2LimitDegrees=60, twistLimitDegrees=3}
 {bone1Name=dyn_shirt_front_l_02, bone2Name=dyn_shirt_front_l_01, swing1LimitDegrees=3, swing2LimitDegrees=10, twistLimitDegrees=3}
 {bone1Name=dyn_shirt_front_r_02, bone2Name=dyn_shirt_front_r_01, swing1LimitDegrees=3, swing2LimitDegrees=10, twistLimitDegrees=3}
 {bone1Name=dyn_shirt_back_r_02, bone2Name=dyn_shirt_back_r_01, swing1LimitDegrees=8, swing2LimitDegrees=60, twistLimitDegrees=4}
 {bone1Name=dyn_shirt_back_l_01, bone2Name=pelvis, swing1LimitDegrees=5, swing2LimitDegrees=60, twistLimitDegrees=3}
 {bone1Name=dyn_shirt_back_l_02, bone2Name=dyn_shirt_back_l_01, swing1LimitDegrees=8, swing2LimitDegrees=60, twistLimitDegrees=3}
 {bone1Name=dyn_l_knee_string_long_01, bone2Name=calf_l, swing1LimitDegrees=10, swing2LimitDegrees=10, twistLimitDegrees=5}
 {bone1Name=dyn_l_knee_string_long_02, bone2Name=dyn_l_knee_string_long_01, swing1LimitDegrees=10, swing2LimitDegrees=10, twistLimitDegrees=5}
 {bone1Name=dyn_l_knee_string_long_03, bone2Name=dyn_l_knee_string_long_02, swing1LimitDegrees=10, swing2LimitDegrees=10, twistLimitDegrees=5}
 {
  bone1Name=thigh_l
  bone2Name=pelvis
  swing1LimitDegrees=45
  swing1Motion=Locked
  swing2LimitDegrees=45
  swing2Motion=Locked
  twistLimitDegrees=45
  twistMotion=Locked
 }
 {
  bone1Name=spine_05
  bone2Name=pelvis
  swing1LimitDegrees=45
  swing1Motion=Locked
  swing2LimitDegrees=45
  swing2Motion=Locked
  twistLimitDegrees=45
  twistMotion=Locked
 }
 {
  bone1Name=clavicle_l
  bone2Name=spine_05
  swing1LimitDegrees=45
  swing1Motion=Free
  swing2LimitDegrees=45
  swing2Motion=Free
  twistLimitDegrees=45
  twistMotion=Free
 }
 {
  bone1Name=clavicle_r
  bone2Name=spine_05
  swing1LimitDegrees=45
  swing1Motion=Free
  swing2LimitDegrees=45
  swing2Motion=Free
  twistLimitDegrees=45
  twistMotion=Free
 }
 {
  bone1Name=upperarm_r
  bone2Name=clavicle_r
  swing1LimitDegrees=45
  swing1Motion=Free
  swing2LimitDegrees=45
  swing2Motion=Free
  twistLimitDegrees=45
  twistMotion=Free
 }
 {bone1Name=dyn_back_strap_01, bone2Name=pelvis, swing1LimitDegrees=5, swing2LimitDegrees=8, twistLimitDegrees=5}
 {bone1Name=dyn_back_strap_02, bone2Name=dyn_back_strap_01, swing1LimitDegrees=15, swing2LimitDegrees=25, twistLimitDegrees=5}
 {bone1Name=dyn_back_strap_03, bone2Name=dyn_back_strap_02, swing1LimitDegrees=15, swing2LimitDegrees=25, twistLimitDegrees=5}
 {
  bone1Name=dyn_shirt_back_r_02
  bone2Name=dyn_shirt_back_l_02
  swing1LimitDegrees=45
  swing1Motion=Free
  swing2LimitDegrees=45
  swing2Motion=Free
  twistLimitDegrees=45
  twistMotion=Free
 }
]
