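# Physics asset for the Wombat medium male character's face-accessory part, giving its dynamic hair bones secondary motion.
# Capsule and box bodies form front, side and back hair chains rooted at the head, each link held by swing and twist limits.
# Head, neck, upper spine and right clavicle shapes join through free angular constraints, as do two extra cross-links tying the back and right chains to the back-right chain.
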PhysicsAsset /BRCosmetics/Characters/Player/Male/Medium/Bodies/M_MED_Wombat/Meshes/Parts/M_MED_Wombat_FaceAcc_Dyn
bodies {
 clavicle_r [{shapeType=Capsule, length=6.6764, radius=7.9902, center=<-8.2296 0.0002 2.1558>, rotation=<rot 78.3434 -0.0002 0>}]
 dyn_hair_bk_01 [{shapeType=Capsule, length=6.0775, radius=2.3728, shapeName=dyn_hair_bk_01_capsule, center=<3.2322 1.4478 0>}]
 dyn_hair_bk_02 [{shapeType=Capsule, length=6.6617, radius=2.6827, shapeName=dyn_hair_bk_02_capsule, center=<2.637 -0.0002 2.3696>, rotation=<rot 0 0 90>}]
 dyn_hair_bk_03 [{shapeType=Capsule, length=3.615, radius=4.3092, shapeName=dyn_hair_bk_03_capsule, center=<4.9833 -0.0002 1.1321>, rotation=<rot 0 0 -90>}]
 dyn_hair_bk_l_01 [{shapeType=Capsule, length=5.3374, radius=3.1129, shapeName=dyn_hair_bk_l_01_capsule, center=<4.6052 0.0001 1.2832>, rotation=<rot 0 0 90>}]
 dyn_hair_bk_l_02 [{
   shapeType=Capsule
   length=6.2934
   radius=3.051
   shapeName=dyn_hair_bk_l_02_capsule
   center=<2.6371 2.4127 -0.0002>
   rotation=<rot 0 0 -31.7472>
  }]
 dyn_hair_bk_l_03 [{
   shapeType=Capsule
   length=11.1111
   radius=4.3033
   shapeName=dyn_hair_bk_l_03_capsule
   center=<2.4349 -0.261 2.5908>
   rotation=<rot 0 0 -95.752>
  }]
 dyn_hair_bk_r_01 [{shapeType=Capsule, length=6.0775, radius=2.3728, shapeName=dyn_hair_bk_r_01_capsule, center=<3.2322 0.0001 1.2832>, rotation=<rot 0 0 90>}]
 dyn_hair_bk_r_02 [{shapeType=Capsule, length=6.6617, radius=2.6827, shapeName=dyn_hair_bk_r_02_capsule, center=<2.637 -0.0002 2.3696>, rotation=<rot 0 0 90>}]
 dyn_hair_bk_r_03 [{shapeType=Capsule, length=7.7484, radius=2, shapeName=dyn_hair_bk_r_03_capsule, center=<2.4806 0 1.1046>, rotation=<rot 0 0 90>}]
 dyn_hair_bk_r_04 [{shapeType=Box, extentX=5.4151, extentY=15.1567, extentZ=2.5896, center=<2.3117 -0.0001 1.4168>, rotation=<rot 20.9747 0 -0.0001>}]
 dyn_hair_fr_01 [{shapeType=Capsule, length=5.0201, radius=1.2563, shapeName=dyn_hair_fr_01_capsule, center=<1.6876 0 0>, rotation=<rot 0 -0.0172 -23.4475>}]
 dyn_hair_fr_02 [{shapeType=Capsule, length=3.3915, radius=1.8603, shapeName=dyn_hair_fr_02_capsule, center=<2.2294 0 0>, rotation=<rot 0 -0.0001 -23.5097>}]
 dyn_hair_fr_04 [{shapeType=Box, extentX=2.6017, extentY=7.5, extentZ=1.7408, center=<1.8441 0 0>, rotation=<rot 0 0 -79.1295>}]
 dyn_hair_fr_05 [{shapeType=Box, extentX=3.1017, extentY=5.8439, extentZ=4.811, center=<4.5291 0 0>, rotation=<rot 0.0007 0 98.5063>}]
 dyn_hair_l_01 [{shapeType=Capsule, length=5.7203, radius=2.73, shapeName=dyn_hair_l_01_capsule, center=<3.2322 0.0001 1.2832>, rotation=<rot 0 0 90>}]
 dyn_hair_l_02 [{shapeType=Capsule, length=6.644, radius=2.7005, shapeName=dyn_hair_l_02_capsule, center=<2.637 -0.0002 2.3696>, rotation=<rot 0 0 90>}]
 dyn_hair_l_03 [{
   shapeType=Capsule
   length=5.406
   radius=4.2529
   shapeName=dyn_hair_l_03_capsule
   center=<2.9559 -0.1365 2.9418>
   rotation=<rot -0.0001 0.0071 -92.6637>
  }]
 dyn_hair_r_01 [{shapeType=Capsule, length=5.9864, radius=2.464, shapeName=dyn_hair_r_01_capsule, center=<3.2322 0.0001 1.2832>, rotation=<rot 0 0 90>}]
 dyn_hair_r_02 [{shapeType=Capsule, length=6.6617, radius=2.6827, shapeName=dyn_hair_r_02_capsule, center=<2.637 -0.0002 2.3696>, rotation=<rot 0 0 90>}]
 dyn_hair_r_03 [
  {shapeType=Capsule, length=6.1704, radius=3.174, shapeName=dyn_hair_l_02_capsule, center=<4.8128 -0.0002 2.0117>, rotation=<rot 0 0 90>}
  {shapeType=Capsule, length=6.8608, radius=2.335, shapeName=dyn_hair_r_03_capsule, center=<2.2422 -2.5796 -0.5749>, rotation=<rot 0 0 55.5294>}
 ]
 head [
  {shapeType=Capsule, length=5.3867, radius=6.6133, center=<3.053 0.2123 0.1132>, rotation=<rot 90 0 0>}
  {shapeType=Capsule, length=5.3867, radius=6.6133, center=<3.0531 3.7709 0.1132>, rotation=<rot 90 0 0>}
  {shapeType=Capsule, length=4.8867, radius=7.1133, center=<4.8911 1.8179 0.1133>, rotation=<rot 90 0 0>}
 ]
 neck_01 [
  {shapeType=Capsule, length=4.8701, radius=5.7072, center=<2.1336 0.9263 0.0001>, rotation=<rot 90 0 13.5236>}
  {shapeType=Capsule, length=3, radius=6.25}
 ]
 spine_05 [{shapeType=Capsule, length=15.6298, radius=13.1663}]
}
constraints [
 {bone1Name=dyn_hair_bk_l_01, bone2Name=head, swing1LimitDegrees=5, swing2LimitDegrees=45, twistLimitDegrees=5}
 {bone1Name=dyn_hair_bk_l_02, bone2Name=dyn_hair_bk_l_01, swing1LimitDegrees=5, swing2LimitDegrees=45, twistLimitDegrees=20}
 {bone1Name=dyn_hair_bk_l_03, bone2Name=dyn_hair_bk_l_02, swing1LimitDegrees=5, swing2LimitDegrees=25, twistLimitDegrees=20}
 {bone1Name=dyn_hair_bk_01, bone2Name=head, swing1LimitDegrees=45, swing2LimitDegrees=5, twistLimitDegrees=5}
 {bone1Name=dyn_hair_bk_02, bone2Name=dyn_hair_bk_01, swing1LimitDegrees=5, swing2LimitDegrees=45, twistLimitDegrees=5}
 {bone1Name=dyn_hair_bk_03, bone2Name=dyn_hair_bk_02, swing1LimitDegrees=5, swing2LimitDegrees=45, twistLimitDegrees=5}
 {bone1Name=dyn_hair_l_01, bone2Name=head, swing1LimitDegrees=5, swing2LimitDegrees=25, twistLimitDegrees=30}
 {bone1Name=dyn_hair_l_02, bone2Name=dyn_hair_l_01, swing1LimitDegrees=5, swing2LimitDegrees=25, twistLimitDegrees=30}
 {bone1Name=dyn_hair_l_03, bone2Name=dyn_hair_l_02, swing1LimitDegrees=5, swing2LimitDegrees=25, twistLimitDegrees=30}
 {bone1Name=dyn_hair_r_01, bone2Name=head, swing1LimitDegrees=5, swing2LimitDegrees=25, twistLimitDegrees=5}
 {bone1Name=dyn_hair_r_02, bone2Name=dyn_hair_r_01, swing1LimitDegrees=5, swing2LimitDegrees=25, twistLimitDegrees=5}
 {bone1Name=dyn_hair_r_03, bone2Name=dyn_hair_r_02, swing1LimitDegrees=5, swing2LimitDegrees=25, twistLimitDegrees=5}
 {bone1Name=dyn_hair_bk_r_01, bone2Name=head, swing1LimitDegrees=5, swing2LimitDegrees=45, twistLimitDegrees=5}
 {bone1Name=dyn_hair_bk_r_02, bone2Name=dyn_hair_bk_r_01, swing1LimitDegrees=5, swing2LimitDegrees=25, twistLimitDegrees=5}
 {bone1Name=dyn_hair_bk_r_03, bone2Name=dyn_hair_bk_r_02, swing1LimitDegrees=5, swing2LimitDegrees=45, twistLimitDegrees=5}
 {bone1Name=dyn_hair_bk_r_04, bone2Name=dyn_hair_bk_r_03, swing1LimitDegrees=5, swing2LimitDegrees=45, twistLimitDegrees=5}
 {
  bone1Name=head
  bone2Name=neck_01
  swing1LimitDegrees=45
  swing1Motion=Free
  swing2LimitDegrees=45
  swing2Motion=Free
  twistLimitDegrees=45
  twistMotion=Free
 }
 {
  bone1Name=neck_01
  bone2Name=spine_05
  swing1LimitDegrees=45
  swing1Motion=Free
  swing2LimitDegrees=45
  swing2Motion=Free
  twistLimitDegrees=45
  twistMotion=Free
 }
 {bone1Name=dyn_hair_fr_01, bone2Name=head, swing1LimitDegrees=15, swing2LimitDegrees=15, twistLimitDegrees=5}
 {bone1Name=dyn_hair_fr_02, bone2Name=dyn_hair_fr_01, swing1LimitDegrees=15, swing2LimitDegrees=15, twistLimitDegrees=5}
 {bone1Name=dyn_hair_fr_04, bone2Name=head, swing1LimitDegrees=15, swing2LimitDegrees=15, twistLimitDegrees=5}
 {bone1Name=dyn_hair_fr_05, bone2Name=dyn_hair_fr_04, swing1LimitDegrees=15, swing2LimitDegrees=15, twistLimitDegrees=5}
 {
  bone1Name=clavicle_r
  bone2Name=spine_05
  swing1LimitDegrees=45
  swing1Motion=Free
  swing2LimitDegrees=45
  swing2Motion=Free
  twistLimitDegrees=45
  twistMotion=Free
 }
 {
  bone1Name=dyn_hair_bk_02
  bone2Name=dyn_hair_bk_r_03
  swing1LimitDegrees=45
  swing1Motion=Free
  swing2LimitDegrees=45
  swing2Motion=Free
  twistLimitDegrees=45
  twistMotion=Free
 }
 {
  bone1Name=dyn_hair_r_03
  bone2Name=dyn_hair_bk_r_03
  swing1LimitDegrees=45
  swing1Motion=Free
  swing2LimitDegrees=45
  swing2Motion=Free
  twistLimitDegrees=45
  twistMotion=Free
 }
]
