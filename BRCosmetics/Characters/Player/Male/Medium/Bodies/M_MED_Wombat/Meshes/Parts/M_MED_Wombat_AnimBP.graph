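# Post-process style AnimBP for a Wombat character part that copies the parent mesh pose and adds secondary hair motion.
# Skydiving selects spring AnimDynamics chains with fixed bone offsets; otherwise RigidBody simulations differ in base bone and inertia for airborne (falling/jumping) versus other states.
# Both RigidBody branches apply additive hair corrections weighted by Aim_Up and Rifle_Adjustment.
# The update event derives those weights from right arm and clavicle rotations.
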AnimBlueprint /BRCosmetics/Characters/Player/Male/Medium/Bodies/M_MED_Wombat/Meshes/Parts/M_MED_Wombat_AnimBP
target_skeleton=&/M_MED_Wombat_Skeleton
graphs [
 AnimGraph {nodes [
   n1 = RigidBody!(
    ComponentPose=Local To Component(LocalPose=Copy Pose From Mesh() {bUseAttachedParent=true})
    Alpha
    OverrideWorldGravity:Vector="0,0,-980"
    ExternalForce
   ) -> (Pose) {
    bClampLinearTranslationLimitToRefPose=true
    bOverrideWorldGravity=true
    overridePhysicsAsset=&/M_MED_Wombat_FaceAcc_Dyn
    simulationSpace=BaseBoneSpace
    alphaBoolBlend {blendOption=Linear}
    baseBoneRef {boneName=spine_05}
    componentAppliedLinearAccClamp=<3000 3000 3000>
    componentLinearAccScale=<0.37 0.37 0.37>
    componentLinearVelScale=<0.1 0.1 0.1>
    overrideWorldGravity {z=-980}
   }
   n2 = Get Aim_Up() -> (Aim_Up:double)
   n3 = "Transform (Modify) Bone - Bone: dyn_hair_r_02"(ComponentPose=@n1, Alpha=@n2, Translation="0,0,0", Rotation="-30,20,15", Scale="1,1,1") {
    rotationMode=BMM_Additive
    scaleMode=BMM_Additive
    translationMode=BMM_Additive
    alphaBoolBlend {blendOption=Linear}
    boneToModify {boneName=dyn_hair_r_02}
    rotation=<rot -30 20 15>
   }
   n4 = "Transform (Modify) Bone - Bone: dyn_hair_bk_r_03"(ComponentPose=@n3, Alpha=@n2, Translation="0,0,0.75", Rotation="-30,0,-20", Scale="1,1,1") {
    rotationMode=BMM_Additive
    scaleMode=BMM_Additive
    translationMode=BMM_Additive
    alphaBoolBlend {blendOption=Linear}
    boneToModify {boneName=dyn_hair_bk_r_03}
    rotation {pitch=-30, roll=-20}
    translation {z=0.75}
   }
   n5 = Get Rifle_Adjustment() -> (Rifle_Adjustment:double)
   n6 = "Transform (Modify) Bone - Bone: dyn_hair_r_02"(ComponentPose=@n4, Alpha=@n5, Translation="0,0,0.5", Rotation="-10,0,0", Scale="1,1,1") {
    rotationMode=BMM_Additive
    scaleMode=BMM_Additive
    translationMode=BMM_Additive
    alphaBoolBlend {blendOption=Linear}
    boneToModify {boneName=dyn_hair_r_02}
    rotation {pitch=-10}
    translation {z=0.5}
   }
   n7 = "Transform (Modify) Bone - Bone: dyn_hair_bk_r_03"(ComponentPose=@n6, Alpha=@n5, Translation="0,0,1.5", Rotation="-15,0,-20", Scale="1,1,1") {
    rotationMode=BMM_Additive
    scaleMode=BMM_Additive
    translationMode=BMM_Additive
    alphaBoolBlend {blendOption=Linear}
    boneToModify {boneName=dyn_hair_bk_r_03}
    rotation {pitch=-15, roll=-20}
    translation {z=1.5}
   }
   n8 = Get Rifle_Adjustment()
   n9 = "Transform (Modify) Bone - Bone: dyn_hair_bk_02"(ComponentPose=@n7, Alpha=@n8, Translation="0,0,0.75", Rotation="0,0,20", Scale="1,1,1") {
    rotationMode=BMM_Additive
    scaleMode=BMM_Additive
    translationMode=BMM_Additive
    alphaBoolBlend {blendOption=Linear}
    boneToModify {boneName=dyn_hair_bk_02}
    rotation {roll=20}
    translation {z=0.75}
   }
   n10 = "Transform (Modify) Bone - Bone: dyn_hair_bk_r_02"(ComponentPose=@n9, Alpha=@n8, Translation="0,0,0", Rotation="-12,0,10", Scale="1,1,1") {
    rotationMode=BMM_Additive
    scaleMode=BMM_Additive
    translationMode=BMM_Additive
    alphaBoolBlend {blendOption=Linear}
    boneToModify {boneName=dyn_hair_bk_r_02}
    rotation {pitch=-12, roll=10}
   }
   n11 = Component To Local(ComponentPose=@n10)
   n12 = RigidBody(ComponentPose=Local To Component(LocalPose=Copy Pose From Mesh() {bUseAttachedParent=true}), OverrideWorldGravity="0,0,-980") {
    bClampLinearTranslationLimitToRefPose=true
    bOverrideWorldGravity=true
    overridePhysicsAsset=&M_MED_Wombat_FaceAcc_Dyn
    simulationSpace=BaseBoneSpace
    alphaBoolBlend {blendOption=Linear}
    baseBoneRef {boneName=neck_02}
    componentAppliedLinearAccClamp=<3000 3000 3000>
    componentLinearAccScale=<0.45 0.25 0.23>
    componentLinearVelScale=<0.35 0.2 0.2>
    overrideWorldGravity {z=-980}
   }
   n13 = Get Aim_Up()
   n14 = "Transform (Modify) Bone - Bone: dyn_hair_r_02"(ComponentPose=@n12, Alpha=@n13, Translation="0,0,0", Rotation="-30,20,15", Scale="1,1,1") {
    rotationMode=BMM_Additive
    scaleMode=BMM_Additive
    translationMode=BMM_Additive
    alphaBoolBlend {blendOption=Linear}
    boneToModify {boneName=dyn_hair_r_02}
    rotation=<rot -30 20 15>
   }
   n15 = "Transform (Modify) Bone - Bone: dyn_hair_bk_r_03"(ComponentPose=@n14, Alpha=@n13, Translation="0,0,0.75", Rotation="-30,0,-20", Scale="1,1,1") {
    rotationMode=BMM_Additive
    scaleMode=BMM_Additive
    translationMode=BMM_Additive
    alphaBoolBlend {blendOption=Linear}
    boneToModify {boneName=dyn_hair_bk_r_03}
    rotation {pitch=-30, roll=-20}
    translation {z=0.75}
   }
   n16 = Get Rifle_Adjustment()
   n17 = "Transform (Modify) Bone - Bone: dyn_hair_r_02"(ComponentPose=@n15, Alpha=@n16, Translation="0,0,0.5", Rotation="-10,0,0", Scale="1,1,1") {
    rotationMode=BMM_Additive
    scaleMode=BMM_Additive
    translationMode=BMM_Additive
    alphaBoolBlend {blendOption=Linear}
    boneToModify {boneName=dyn_hair_r_02}
    rotation {pitch=-10}
    translation {z=0.5}
   }
   n18 = "Transform (Modify) Bone - Bone: dyn_hair_bk_r_03"(ComponentPose=@n17, Alpha=@n16, Translation="0,0,1.5", Rotation="-15,0,-20", Scale="1,1,1") {
    rotationMode=BMM_Additive
    scaleMode=BMM_Additive
    translationMode=BMM_Additive
    alphaBoolBlend {blendOption=Linear}
    boneToModify {boneName=dyn_hair_bk_r_03}
    rotation {pitch=-15, roll=-20}
    translation {z=1.5}
   }
   n19 = Get Rifle_Adjustment()
   n20 = "Transform (Modify) Bone - Bone: dyn_hair_bk_02"(ComponentPose=@n18, Alpha=@n19, Translation="0,0,0.75", Rotation="0,0,20", Scale="1,1,1") {
    rotationMode=BMM_Additive
    scaleMode=BMM_Additive
    translationMode=BMM_Additive
    alphaBoolBlend {blendOption=Linear}
    boneToModify {boneName=dyn_hair_bk_02}
    rotation {roll=20}
    translation {z=0.75}
   }
   n21 = "Transform (Modify) Bone - Bone: dyn_hair_bk_r_02"(ComponentPose=@n20, Alpha=@n19, Translation="0,0,0", Rotation="-12,0,10", Scale="1,1,1") {
    rotationMode=BMM_Additive
    scaleMode=BMM_Additive
    translationMode=BMM_Additive
    alphaBoolBlend {blendOption=Linear}
    boneToModify {boneName=dyn_hair_bk_r_02}
    rotation {pitch=-12, roll=10}
   }
   n22 = Component To Local(ComponentPose=@n21)
   n23 = OR Boolean(A=Get bIsFalling(), B=Get bIsJumping())
   n24 = "Transform (Modify) Bone - Bone: dyn_hair_bk_l_01"(
    ComponentPose=Local To Component(LocalPose=Copy Pose From Mesh() {bUseAttachedParent=true})
    Alpha=1
    Translation="0,0,2"
    Rotation="20,0,0"
    Scale="1,1,1"
   ) {
    rotationMode=BMM_Additive
    rotationSpace=BCS_BoneSpace
    scaleMode=BMM_Additive
    scaleSpace=BCS_BoneSpace
    translationMode=BMM_Additive
    translationSpace=BCS_BoneSpace
    alphaBoolBlend {blendOption=Linear}
    boneToModify {boneName=dyn_hair_bk_l_01}
    rotation {pitch=20}
    translation {z=2}
   }
   n25 = Anim Dynamics - Bone: dyn_hair_bk_l_02!(ComponentPose=@n24, Alpha=1, GravityOverride:Vector="0,0,0", ExternalForce="0,0,0") -> (Pose) {
    angularDampingOverride=0.9999
    angularSpringConstant=2500
    bAngularSpring=true
    bOverrideAngularDamping=true
    bUseGravityOverride=true
    simulationSpace=RootRelative
    alphaBoolBlend {blendOption=Linear}
    boundBone {boneName=dyn_hair_bk_l_02}
    componentAppliedLinearAccClamp=<5000 5000 5000>
    componentLinearAccScale=<0.5 0.5 0.5>
    componentLinearVelScale=<0.5 0.5 0.5>
    physicsBodyDefinitions [{
      sphereCollisionRadius=0
      boundBone {boneName=dyn_hair_bk_l_02}
      boxExtents=<10 20 20>
      constraintSetup {angularLimitsMax=<30 10 30>, angularLimitsMin=<-30 -45 -30>, angularTarget=<0 0 0>}
      localJointOffset=<2 0 0>
     }]
   }
   n26 = Anim Dynamics - Bone: dyn_hair_bk_l_03!(ComponentPose=@n25, Alpha=1, GravityOverride:Vector="0,0,0", ExternalForce="0,0,0") -> (Pose) {
    angularDampingOverride=0.9999
    angularSpringConstant=2500
    bAngularSpring=true
    bOverrideAngularDamping=true
    bUseGravityOverride=true
    simulationSpace=RootRelative
    alphaBoolBlend {blendOption=Linear}
    boundBone {boneName=dyn_hair_bk_l_03}
    componentAppliedLinearAccClamp=<5000 5000 5000>
    componentLinearAccScale=<0.5 0.5 0.5>
    componentLinearVelScale=<0.5 0.5 0.5>
    physicsBodyDefinitions [{
      sphereCollisionRadius=0
      boundBone {boneName=dyn_hair_bk_l_03}
      boxExtents=<10 20 20>
      constraintSetup {angularLimitsMax=<30 10 30>, angularLimitsMin=<-30 -45 -30>, angularTarget=<0 0 0>}
      localJointOffset=<2 0 0>
     }]
   }
   n27 = "Transform (Modify) Bone - Bone: dyn_hair_bk_01"(ComponentPose=@n26, Alpha=1, Translation="0,2,0", Rotation="0,25,0", Scale="1,1,1") {
    rotationMode=BMM_Additive
    rotationSpace=BCS_BoneSpace
    scaleMode=BMM_Additive
    scaleSpace=BCS_BoneSpace
    translationMode=BMM_Additive
    translationSpace=BCS_BoneSpace
    alphaBoolBlend {blendOption=Linear}
    boneToModify {boneName=dyn_hair_bk_01}
    rotation {yaw=25}
    translation {y=2}
   }
   n28 = Anim Dynamics - Bone: dyn_hair_bk_02!(ComponentPose=@n27, Alpha=1, GravityOverride:Vector="0,0,0", ExternalForce="0,0,0") -> (Pose) {
    angularDampingOverride=0.9999
    angularSpringConstant=2500
    bAngularSpring=true
    bOverrideAngularDamping=true
    bUseGravityOverride=true
    simulationSpace=RootRelative
    alphaBoolBlend {blendOption=Linear}
    boundBone {boneName=dyn_hair_bk_02}
    componentAppliedLinearAccClamp=<5000 5000 5000>
    componentLinearAccScale=<0.5 0.5 0.5>
    componentLinearVelScale=<0.5 0.5 0.5>
    physicsBodyDefinitions [{
      sphereCollisionRadius=0
      boundBone {boneName=dyn_hair_bk_02}
      boxExtents=<10 20 20>
      constraintSetup {angularLimitsMax=<30 10 30>, angularLimitsMin=<-30 -45 -30>, angularTarget=<0 0 0>}
      localJointOffset=<2 0 0>
     }]
   }
   n29 = Anim Dynamics - Bone: dyn_hair_bk_03!(ComponentPose=@n28, Alpha=1, GravityOverride:Vector="0,0,0", ExternalForce="0,0,0") -> (Pose) {
    angularDampingOverride=0.9999
    angularSpringConstant=2500
    bAngularSpring=true
    bOverrideAngularDamping=true
    bUseGravityOverride=true
    simulationSpace=RootRelative
    alphaBoolBlend {blendOption=Linear}
    boundBone {boneName=dyn_hair_bk_03}
    componentAppliedLinearAccClamp=<5000 5000 5000>
    componentLinearAccScale=<0.5 0.5 0.5>
    componentLinearVelScale=<0.5 0.5 0.5>
    physicsBodyDefinitions [{
      sphereCollisionRadius=0
      boundBone {boneName=dyn_hair_bk_03}
      boxExtents=<10 20 20>
      constraintSetup {angularLimitsMax=<30 10 30>, angularLimitsMin=<-30 -45 -30>, angularTarget=<0 0 0>}
      localJointOffset=<2 0 0>
     }]
   }
   n30 = "Transform (Modify) Bone - Bone: dyn_hair_bk_r_01"(ComponentPose=@n29, Alpha=1, Translation="0,0,2", Rotation="20,0,0", Scale="1,1,1") {
    rotationMode=BMM_Additive
    rotationSpace=BCS_BoneSpace
    scaleMode=BMM_Additive
    scaleSpace=BCS_BoneSpace
    translationMode=BMM_Additive
    translationSpace=BCS_BoneSpace
    alphaBoolBlend {blendOption=Linear}
    boneToModify {boneName=dyn_hair_bk_r_01}
    rotation {pitch=20}
    translation {z=2}
   }
   n31 = Anim Dynamics - Bone: dyn_hair_bk_r_02!(ComponentPose=@n30, Alpha=1, GravityOverride:Vector="0,0,0", ExternalForce="0,0,0") -> (Pose) {
    angularDampingOverride=0.9999
    angularSpringConstant=2500
    bAngularSpring=true
    bOverrideAngularDamping=true
    bUseGravityOverride=true
    simulationSpace=RootRelative
    alphaBoolBlend {blendOption=Linear}
    boundBone {boneName=dyn_hair_bk_r_02}
    componentAppliedLinearAccClamp=<5000 5000 5000>
    componentLinearAccScale=<0.5 0.5 0.5>
    componentLinearVelScale=<0.5 0.5 0.5>
    physicsBodyDefinitions [{
      sphereCollisionRadius=0
      boundBone {boneName=dyn_hair_bk_r_02}
      boxExtents=<10 20 20>
      constraintSetup {angularLimitsMax=<30 10 30>, angularLimitsMin=<-30 -45 -30>, angularTarget=<0 0 0>}
      localJointOffset=<2 0 0>
     }]
   }
   n32 = Anim Dynamics - Bone: dyn_hair_bk_r_03!(ComponentPose=@n31, Alpha=1, GravityOverride:Vector="0,0,0", ExternalForce="0,0,0") -> (Pose) {
    angularDampingOverride=0.9999
    angularSpringConstant=2500
    bAngularSpring=true
    bOverrideAngularDamping=true
    bUseGravityOverride=true
    simulationSpace=RootRelative
    alphaBoolBlend {blendOption=Linear}
    boundBone {boneName=dyn_hair_bk_r_03}
    componentAppliedLinearAccClamp=<5000 5000 5000>
    componentLinearAccScale=<0.5 0.5 0.5>
    componentLinearVelScale=<0.5 0.5 0.5>
    physicsBodyDefinitions [{
      sphereCollisionRadius=0
      boundBone {boneName=dyn_hair_bk_r_03}
      boxExtents=<10 20 20>
      constraintSetup {angularLimitsMax=<30 10 30>, angularLimitsMin=<-30 -45 -30>, angularTarget=<0 0 0>}
      localJointOffset=<2 0 0>
     }]
   }
   n33 = Anim Dynamics - Bone: dyn_hair_bk_r_04!(ComponentPose=@n32, Alpha=1, GravityOverride:Vector="0,0,0", ExternalForce="0,0,0") -> (Pose) {
    angularDampingOverride=0.9999
    angularSpringConstant=2500
    bAngularSpring=true
    bOverrideAngularDamping=true
    bUseGravityOverride=true
    simulationSpace=RootRelative
    alphaBoolBlend {blendOption=Linear}
    boundBone {boneName=dyn_hair_bk_r_04}
    componentAppliedLinearAccClamp=<5000 5000 5000>
    componentLinearAccScale=<0.5 0.5 0.5>
    componentLinearVelScale=<0.5 0.5 0.5>
    physicsBodyDefinitions [{
      sphereCollisionRadius=0
      boundBone {boneName=dyn_hair_bk_r_04}
      boxExtents=<10 20 20>
      constraintSetup {angularLimitsMax=<30 10 30>, angularLimitsMin=<-30 -45 -30>, angularTarget=<0 0 0>}
      localJointOffset=<2 0 0>
     }]
   }
   Output Pose(Result=Blend Poses by bool(
     BlendPose_0=Component To Local(ComponentPose=@n33)
     BlendPose_1=Blend Poses by bool(BlendPose_0=@n11, BlendPose_1=@n22, BlendTime_0=0.3, bActiveValue=@n23) {blendType=Linear}
     bActiveValue=Get bIsSkyDiving()
    ) {blendType=Linear}) {layerGroup=None}
  ]}
 EventGraph {nodes [
   Event BlueprintUpdateAnimation()
   > n1 = Sequence()
   n2 = GetOwnerSkeletalMeshComponent()
   n3 = BreakTransform(InTransform=GetDeltaTransformFromRefPose(self=@n2, BoneName=upperarm_r, BaseName=clavicle_r))
   Set Aim_Up(execute=@n1.then_1, Aim_Up:double=MapRangeClamped(Value=BreakRotator(InRot=@n3.Rotation).Yaw, InRangeA=90, InRangeB=120)) -> (Output_Get:double)
   n4 = BreakTransform(InTransform=GetDeltaTransformFromRefPose(self=@n2, BoneName=clavicle_r, BaseName=spine_05))
   > Set Rifle_Adjustment(Rifle_Adjustment:double=MapRangeClamped(Value=BreakRotator(InRot=@n4.Rotation), InRangeA=-2, InRangeB=-4)) -> (Output_Get:double)
   n5 = Gravity Override State Function DEPRECATED(
    execute=@n1.then_0
    CharacterPartAnimInstance=Self-Reference()
    GameTimeInSeconds=GetGameTimeInSeconds()
    GravityOverrideParams=Get Gravity Override Params()
   )
   > Set GravityOverride(GravityOverride=@n5)
  ]}
]
variables [
 Aim_Up {type=number}
 Gravity Override Params {type=GravityOverrideParamsStruct}
 GravityOverride {type=Vector}
 Head_Adjust {type=number}
 Rifle_Adjustment {type=number}
]
