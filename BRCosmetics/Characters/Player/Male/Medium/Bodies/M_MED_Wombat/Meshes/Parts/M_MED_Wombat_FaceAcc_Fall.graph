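# Physics asset for the Fall face accessory of the Wombat medium male character, covering its dynamic hair bones.
# Box bodies form short front, side and back chains hanging from the head, with swing and twist limits.
# Capsules on head, neck, upper spine and right clavicle supply body shapes, joined by angularly free constraints.
# Extra angularly free constraints cross-link neighboring back and left-side chains, tying them into a connected sheet.
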PhysicsAsset /BRCosmetics/Characters/Player/Male/Medium/Bodies/M_MED_Wombat/Meshes/Parts/M_MED_Wombat_FaceAcc_Fall
bodies {
 clavicle_r [{shapeType=Capsule, length=6.6764, radius=6.7402, center=<-8.6744 0.0004 0>, rotation=<rot 90 0 0>}]
 dyn_hair_bk_01 [{shapeType=Box, extentX=6.1017, extentY=7.5, extentZ=1.7408, center=<3.3857 0 0>, rotation=<rot 0 0 -93.0265>}]
 dyn_hair_bk_02 [{shapeType=Box, extentX=6.1017, extentY=12.3439, extentZ=1.0881, center=<3.3857 0 0>, rotation=<rot 7.1132 0 0>}]
 dyn_hair_bk_03 [{shapeType=Box, extentX=7.6797, extentY=12.3439, extentZ=4.0348, center=<4.989 0 -0.0585>, rotation=<rot 14.7705 0 0>}]
 dyn_hair_bk_l_01 [{shapeType=Box, extentX=6.1017, extentY=7.5, extentZ=0.856, center=<3.3857 0 0>}]
 dyn_hair_bk_l_02 [{shapeType=Box, extentX=6.1017, extentY=14.4821, extentZ=1.856, center=<3.3857 0 0>, rotation=<rot 0 0 55.3989>}]
 dyn_hair_bk_l_03 [{shapeType=Box, extentX=6.8917, extentY=18.4232, extentZ=3.0124, center=<4.8128 0.0003 0>, rotation=<rot 5.2384 0 0>}]
 dyn_hair_bk_r_01 [{shapeType=Box, extentX=6.1017, extentY=7.5, extentZ=2.2289, center=<3.3857 0 0>}]
 dyn_hair_bk_r_02 [{shapeType=Box, extentX=6.1017, extentY=7.5, extentZ=0.856, center=<3.3857 0 0>}]
 dyn_hair_bk_r_03 [{shapeType=Box, extentX=7.625, extentY=14.3569, extentZ=0.856, center=<3.6212 -0.0001 0.2535>, rotation=<rot 10.0771 0 0>}]
 dyn_hair_bk_r_04 [{shapeType=Box, extentX=5.4151, extentY=15.1567, extentZ=3.5896, center=<2.4392 -0.0001 1.5886>, rotation=<rot 40.9803 -0.0001 -0.0001>}]
 dyn_hair_fr_01 [{shapeType=Box, extentX=2.6017, extentY=7.5, extentZ=1.7408, center=<1.8441 0 0>, rotation=<rot 0 0 -107.5123>}]
 dyn_hair_fr_02 [{shapeType=Box, extentX=3.1017, extentY=5.8439, extentZ=2.1269, center=<1.579 0 0>, rotation=<rot 0.0007 0 73.3082>}]
 dyn_hair_fr_04 [{shapeType=Box, extentX=2.6017, extentY=7.5, extentZ=1.7408, center=<1.8441 0 0>, rotation=<rot 0 0 -79.1295>}]
 dyn_hair_fr_05 [{shapeType=Box, extentX=3.1017, extentY=5.8439, extentZ=2.1269, center=<1.579 0 0>, rotation=<rot 0.0007 0 98.5063>}]
 dyn_hair_l_01 [{shapeType=Box, extentX=6.1017, extentY=4.794, extentZ=0.856, center=<3.2306 0 0>}]
 dyn_hair_l_02 [{shapeType=Box, extentX=6.1017, extentY=8.0419, extentZ=1.356, center=<3.3857 0 0>}]
 dyn_hair_l_03 [{shapeType=Box, extentX=7.449, extentY=12.4356, extentZ=0.856, center=<4.3623 -0.0001 0>}]
 dyn_hair_r_01 [{shapeType=Box, extentX=3.9772, extentY=7.5, extentZ=0.856, center=<2.9504 0.0001 0>}]
 dyn_hair_r_02 [{shapeType=Box, extentX=6.1017, extentY=10.2587, extentZ=0.856, center=<3.3857 0 0.4819>}]
 dyn_hair_r_03 [{shapeType=Box, extentX=6.1017, extentY=10.2587, extentZ=0.856, center=<3.3857 0 0>}]
 head [
  {shapeType=Capsule, length=5.3867, radius=6.6133, center=<3.053 0.2123 0.1132>, rotation=<rot 90 0 0>}
  {shapeType=Capsule, length=5.3867, radius=6.6133, center=<3.0531 3.7709 0.1132>, rotation=<rot 90 0 0>}
  {shapeType=Capsule, length=2.8757, radius=1.2271, center=<2.3942 -0.6886 -6.2403>, rotation=<rot 85.1399 0 0>}
  {shapeType=Capsule, length=2.8757, radius=1.2271, center=<1.9115 -0.7859 6.348>, rotation=<rot 80.2175 180 180>}
  {shapeType=Capsule, length=5.3867, radius=6.6133, center=<4.8911 1.8179 0.1133>, rotation=<rot 90 0 0>}
 ]
 neck_01 [{shapeType=Capsule, length=3.8701, radius=6.7072, center=<2.1336 0.9263 0.0001>, rotation=<rot 90 0 13.5236>}]
 spine_05 [{shapeType=Capsule, length=25.6298, radius=8.6663, center=<2.3756 -0.0002 0>}]
}
constraints [
 {bone1Name=dyn_hair_bk_l_01, bone2Name=head, swing1LimitDegrees=25, swing2LimitDegrees=45, twistLimitDegrees=5}
 {bone1Name=dyn_hair_bk_l_02, bone2Name=dyn_hair_bk_l_01, swing1LimitDegrees=45, swing2LimitDegrees=45, twistLimitDegrees=5}
 {bone1Name=dyn_hair_bk_l_03, bone2Name=dyn_hair_bk_l_02, swing1LimitDegrees=25, swing2LimitDegrees=25, twistLimitDegrees=5}
 {bone1Name=dyn_hair_bk_01, bone2Name=head, swing1LimitDegrees=25, swing2LimitDegrees=45, twistLimitDegrees=5}
 {bone1Name=dyn_hair_bk_02, bone2Name=dyn_hair_bk_01, swing1LimitDegrees=45, swing2LimitDegrees=45, twistLimitDegrees=5}
 {bone1Name=dyn_hair_bk_03, bone2Name=dyn_hair_bk_02, swing1LimitDegrees=25, swing2LimitDegrees=35, twistLimitDegrees=5}
 {bone1Name=dyn_hair_l_01, bone2Name=head, swing1LimitDegrees=25, swing2LimitDegrees=25, twistLimitDegrees=5}
 {bone1Name=dyn_hair_l_02, bone2Name=dyn_hair_l_01, swing1LimitDegrees=25, swing2LimitDegrees=25, twistLimitDegrees=5}
 {bone1Name=dyn_hair_l_03, bone2Name=dyn_hair_l_02, swing1LimitDegrees=25, swing2LimitDegrees=25, twistLimitDegrees=5}
 {bone1Name=dyn_hair_r_01, bone2Name=head, swing1LimitDegrees=25, swing2LimitDegrees=25, twistLimitDegrees=5}
 {bone1Name=dyn_hair_r_02, bone2Name=dyn_hair_r_01, swing1LimitDegrees=25, swing2LimitDegrees=25, twistLimitDegrees=5}
 {bone1Name=dyn_hair_r_03, bone2Name=dyn_hair_r_02, swing1LimitDegrees=25, swing2LimitDegrees=25, twistLimitDegrees=5}
 {bone1Name=dyn_hair_bk_r_01, bone2Name=head, swing1LimitDegrees=25, swing2LimitDegrees=45, twistLimitDegrees=5}
 {bone1Name=dyn_hair_bk_r_02, bone2Name=dyn_hair_bk_r_01, swing1LimitDegrees=25, swing2LimitDegrees=25, twistLimitDegrees=5}
 {bone1Name=dyn_hair_bk_r_03, bone2Name=dyn_hair_bk_r_02, swing1LimitDegrees=45, swing2LimitDegrees=45, twistLimitDegrees=5}
 {bone1Name=dyn_hair_bk_r_04, bone2Name=dyn_hair_bk_r_03, swing1LimitDegrees=25, swing2LimitDegrees=25, twistLimitDegrees=5}
 {
  bone1Name=head
  bone2Name=neck_01
  swing1LimitDegrees=45
  swing1Motion=Free
  swing2LimitDegrees=45
  swing2Motion=Free
  twistLimitDegrees=45
  twistMotion=Free
 }
 {
  bone1Name=neck_01
  bone2Name=spine_05
  swing1LimitDegrees=45
  swing1Motion=Free
  swing2LimitDegrees=45
  swing2Motion=Free
  twistLimitDegrees=45
  twistMotion=Free
 }
 {bone1Name=dyn_hair_fr_01, bone2Name=head, swing1LimitDegrees=25, swing2LimitDegrees=25, twistLimitDegrees=5}
 {bone1Name=dyn_hair_fr_02, bone2Name=dyn_hair_fr_01, swing1LimitDegrees=25, swing2LimitDegrees=25, twistLimitDegrees=5}
 {bone1Name=dyn_hair_fr_04, bone2Name=head, swing1LimitDegrees=40, swing2LimitDegrees=40, twistLimitDegrees=8}
 {bone1Name=dyn_hair_fr_05, bone2Name=dyn_hair_fr_04, swing1LimitDegrees=40, swing2LimitDegrees=40, twistLimitDegrees=8}
 {
  bone1Name=dyn_hair_bk_l_02
  bone2Name=dyn_hair_l_02
  swing1LimitDegrees=45
  swing1Motion=Free
  swing2LimitDegrees=45
  swing2Motion=Free
  twistLimitDegrees=45
  twistMotion=Free
 }
 {
  bone1Name=dyn_hair_bk_l_03
  bone2Name=dyn_hair_bk_03
  swing1LimitDegrees=45
  swing1Motion=Free
  swing2LimitDegrees=45
  swing2Motion=Free
  twistLimitDegrees=45
  twistMotion=Free
 }
 {
  bone1Name=clavicle_r
  bone2Name=spine_05
  swing1LimitDegrees=45
  swing1Motion=Free
  swing2LimitDegrees=45
  swing2Motion=Free
  twistLimitDegrees=45
  twistMotion=Free
 }
 {
  bone1Name=dyn_hair_bk_03
  bone2Name=dyn_hair_bk_r_04
  swing1LimitDegrees=45
  swing1Motion=Free
  swing2LimitDegrees=45
  swing2Motion=Free
  twistLimitDegrees=45
  twistMotion=Free
 }
 {
  bone1Name=dyn_hair_l_03
  bone2Name=dyn_hair_bk_l_03
  swing1LimitDegrees=45
  swing1Motion=Free
  swing2LimitDegrees=45
  swing2Motion=Free
  twistLimitDegrees=45
  twistMotion=Free
 }
 {
  bone1Name=dyn_hair_bk_02
  bone2Name=dyn_hair_bk_r_03
  swing1LimitDegrees=45
  swing1Motion=Free
  swing2LimitDegrees=45
  swing2Motion=Free
  twistLimitDegrees=45
  twistMotion=Free
 }
]
