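# Physics asset for the dynamic face-accessory poncho part of the male medium Wombat player body, rigged on dyn_hair bones.
# Single capsules on front, side and back bones form short chains rooted at the head, with narrow swing and twist limits.
# Larger head and neck capsules provide body volumes, joined by an angularly free constraint.
# Another angularly free constraint cross-links the right side chain to the back-right chain.
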PhysicsAsset /BRCosmetics/Characters/Player/Male/Medium/Bodies/M_MED_Wombat/Meshes/Parts/M_MED_Wombat_FaceAcc_Poncho_Dyn
bodies {
 dyn_hair_bk_03 [{shapeType=Capsule, length=11.0222, radius=0.7734, shapeName=dyn_hair_bk_l_01_capsule, center=<2.4675 0 0>, rotation=<rot 90 0 -90>}]
 dyn_hair_bk_l_02 [{
   shapeType=Capsule
   length=7.372
   radius=0.7382
   shapeName=dyn_hair_bk_l_01_capsule
   center=<3.713 -0.059 0.0872>
   rotation=<rot 0.9235 -0.6206 -33.9003>
  }]
 dyn_hair_bk_l_03 [{
   shapeType=Capsule
   length=10.7205
   radius=0.94
   shapeName=dyn_hair_bk_l_01_capsule
   center=<3.6156 0 0.0001>
   rotation=<rot 90 -26.565 -116.565>
  }]
 dyn_hair_bk_r_02 [{
   shapeType=Capsule
   length=5.5992
   radius=0.5001
   shapeName=dyn_hair_bk_l_01_capsule
   center=<3.7886 -0.0552 0.7152>
   rotation=<rot 90 90 -2.2898>
  }]
 dyn_hair_bk_r_03 [{shapeType=Capsule, length=8.1389, radius=0.8097, shapeName=dyn_hair_bk_l_01_capsule, center=<2.4831 0 0>, rotation=<rot 90 0 -90>}]
 dyn_hair_bk_r_04 [{shapeType=Capsule, length=13.7482, radius=1.2453, shapeName=dyn_hair_bk_l_01_capsule, center=<2.4831 0 0>, rotation=<rot 90 0 -90>}]
 dyn_hair_fr_01 [{shapeType=Capsule, length=6.386, radius=0.8097, shapeName=dyn_hair_bk_l_01_capsule, center=<1.9449 0.0001 0>, rotation=<rot 0 0 -15.898>}]
 dyn_hair_fr_02 [{shapeType=Capsule, length=6.386, radius=0.8097, shapeName=dyn_hair_bk_l_01_capsule, center=<1.9449 0.0001 0>, rotation=<rot 0 0 -15.8891>}]
 dyn_hair_fr_04 [{shapeType=Capsule, length=6.0733, radius=0.8097, shapeName=dyn_hair_bk_l_01_capsule, center=<1.6083 0.0004 0>, rotation=<rot 0 0 7.412>}]
 dyn_hair_fr_05 [{shapeType=Capsule, length=4.1128, radius=1.2011, shapeName=dyn_hair_bk_l_01_capsule, center=<1.6083 0.0004 0>, rotation=<rot 0 0 7.9055>}]
 dyn_hair_l_01 [{shapeType=Capsule, length=4.3749, radius=0.5001, shapeName=dyn_hair_bk_l_01_capsule, center=<2.7573 0 -0.0001>, rotation=<rot 90 0 -90>}]
 dyn_hair_l_02 [{shapeType=Capsule, length=8.3833, radius=0.693, shapeName=dyn_hair_bk_l_01_capsule, center=<2.5796 0 0>, rotation=<rot 90 0 -90>}]
 dyn_hair_l_03 [{
   shapeType=Capsule
   length=11.4156
   radius=0.8208
   shapeName=dyn_hair_bk_l_01_capsule
   center=<3.1926 0 0>
   rotation=<rot 90 -26.565 -116.565>
  }]
 dyn_hair_r_01 [{shapeType=Capsule, length=6.5001, radius=0.8097, shapeName=dyn_hair_bk_l_01_capsule, center=<3.199 0 -0.0001>, rotation=<rot 90 0 -90>}]
 dyn_hair_r_02 [{
   shapeType=Capsule
   length=8.1389
   radius=0.8097
   shapeName=dyn_hair_bk_l_01_capsule
   center=<2.812 0 0.9089>
   rotation=<rot 90 -26.565 -116.565>
  }]
 dyn_hair_r_03 [{shapeType=Capsule, length=8.1389, radius=0.8097, shapeName=dyn_hair_bk_l_01_capsule, center=<3.0382 0 0.2174>, rotation=<rot 90 0 -90>}]
 head [
  {shapeType=Capsule, length=5.3867, radius=6.6133, center=<3.053 0.2123 0.1132>, rotation=<rot 90 0 0>}
  {shapeType=Capsule, length=5.3867, radius=6.6133, center=<3.0531 3.7709 0.1132>, rotation=<rot 90 0 0>}
 ]
 neck_01 [{shapeType=Capsule, length=5.6201, radius=4.9572, center=<2.1336 0.9263 0.0001>, rotation=<rot 90 0 13.5236>}]
}
constraints [
 {bone1Name=dyn_hair_bk_l_02, bone2Name=head, swing1LimitDegrees=45, swing2LimitDegrees=45, twistLimitDegrees=5}
 {bone1Name=dyn_hair_bk_l_03, bone2Name=dyn_hair_bk_l_02, swing1LimitDegrees=25, swing2LimitDegrees=25, twistLimitDegrees=5}
 {bone1Name=dyn_hair_bk_03, bone2Name=head, swing1LimitDegrees=45, swing2LimitDegrees=45, twistLimitDegrees=5}
 {bone1Name=dyn_hair_l_01, bone2Name=head, swing1LimitDegrees=25, swing2LimitDegrees=25, twistLimitDegrees=5}
 {bone1Name=dyn_hair_l_02, bone2Name=dyn_hair_l_01, swing1LimitDegrees=25, swing2LimitDegrees=25, twistLimitDegrees=8}
 {bone1Name=dyn_hair_l_03, bone2Name=dyn_hair_l_02, swing1LimitDegrees=25, swing2LimitDegrees=25, twistLimitDegrees=8}
 {bone1Name=dyn_hair_r_01, bone2Name=head, swing1LimitDegrees=25, swing2LimitDegrees=25, twistLimitDegrees=5}
 {bone1Name=dyn_hair_r_02, bone2Name=dyn_hair_r_01, swing1LimitDegrees=25, swing2LimitDegrees=25, twistLimitDegrees=5}
 {bone1Name=dyn_hair_r_03, bone2Name=dyn_hair_r_02, swing1LimitDegrees=25, swing2LimitDegrees=25, twistLimitDegrees=5}
 {bone1Name=dyn_hair_bk_r_02, bone2Name=head, swing1LimitDegrees=25, swing2LimitDegrees=25, twistLimitDegrees=5}
 {bone1Name=dyn_hair_bk_r_03, bone2Name=dyn_hair_bk_r_02, swing1LimitDegrees=45, swing2LimitDegrees=45, twistLimitDegrees=8}
 {bone1Name=dyn_hair_bk_r_04, bone2Name=dyn_hair_bk_r_03, swing1LimitDegrees=45, swing2LimitDegrees=45, twistLimitDegrees=5}
 {
  bone1Name=head
  bone2Name=neck_01
  swing1LimitDegrees=45
  swing1Motion=Free
  swing2LimitDegrees=45
  swing2Motion=Free
  twistLimitDegrees=45
  twistMotion=Free
 }
 {bone1Name=dyn_hair_fr_01, bone2Name=head, swing1LimitDegrees=25, swing2LimitDegrees=25, twistLimitDegrees=5}
 {bone1Name=dyn_hair_fr_02, bone2Name=dyn_hair_fr_01, swing1LimitDegrees=25, swing2LimitDegrees=25, twistLimitDegrees=5}
 {bone1Name=dyn_hair_fr_04, bone2Name=head, swing1LimitDegrees=40, swing2LimitDegrees=40, twistLimitDegrees=8}
 {bone1Name=dyn_hair_fr_05, bone2Name=dyn_hair_fr_04, swing1LimitDegrees=40, swing2LimitDegrees=40, twistLimitDegrees=8}
 {
  bone1Name=dyn_hair_r_03
  bone2Name=dyn_hair_bk_r_03
  swing1LimitDegrees=45
  swing1Motion=Free
  swing2LimitDegrees=45
  swing2Motion=Free
  twistLimitDegrees=45
  twistMotion=Free
 }
]
