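# Post-process-style secondary-motion AnimBP for the Wombat poncho face accessory, driving its dyn_hair bones.
# The output path copies the parent mesh pose, simulates RigidBody physics in neck_02 space, then constrains bones to head and neck at varying alphas.
# The update event derives Aim_Up and Rifle_Adjustment from arm and clavicle rotations.
# Their additive bone corrections, and a falling/jumping blend with a separate physics asset, do not reach the output.
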AnimBlueprint /BRCosmetics/Characters/Player/Male/Medium/Bodies/M_MED_Wombat/Meshes/Parts/M_MED_Wombat_Poncho_FaceAcc_AnimBP
target_skeleton=&/M_MED_Wombat_Skeleton
graphs [
 AnimGraph {nodes [
   n1 = RigidBody(ComponentPose=Local To Component(LocalPose=Copy Pose From Mesh() {bUseAttachedParent=true})) {
    bClampLinearTranslationLimitToRefPose=true
    bOverrideWorldGravity=true
    overridePhysicsAsset=&/M_MED_Wombat_FaceAcc_Fall
    simulationSpace=BaseBoneSpace
    alphaBoolBlend {blendOption=Linear}
    baseBoneRef {boneName=spine_05}
    componentAppliedLinearAccClamp=<3000 3000 3000>
    componentLinearAccScale=<0.45 0.25 0.23>
    componentLinearVelScale=<0.35 0.2 0.2>
    overrideWorldGravity {z=-980}
   }
   n2 = Get Aim_Up() -> (Aim_Up:double)
   n3 = "Transform (Modify) Bone - Bone: dyn_hair_r_02"(ComponentPose=@n1, Alpha=@n2, Translation="0,0,0", Rotation="-30,20,15", Scale="1,1,1") {
    rotationMode=BMM_Additive
    scaleMode=BMM_Additive
    translationMode=BMM_Additive
    alphaBoolBlend {blendOption=Linear}
    boneToModify {boneName=dyn_hair_r_02}
    rotation=<rot -30 20 15>
   }
   n4 = "Transform (Modify) Bone - Bone: dyn_hair_bk_r_03"(ComponentPose=@n3, Alpha=@n2, Translation="0,0,0.75", Rotation="-30,0,-20", Scale="1,1,1") {
    rotationMode=BMM_Additive
    scaleMode=BMM_Additive
    translationMode=BMM_Additive
    alphaBoolBlend {blendOption=Linear}
    boneToModify {boneName=dyn_hair_bk_r_03}
    rotation {pitch=-30, roll=-20}
    translation {z=0.75}
   }
   n5 = Get Rifle_Adjustment() -> (Rifle_Adjustment:double)
   n6 = "Transform (Modify) Bone - Bone: dyn_hair_r_02"(ComponentPose=@n4, Alpha=@n5, Translation="0,0,0.5", Rotation="-10,0,0", Scale="1,1,1") {
    rotationMode=BMM_Additive
    scaleMode=BMM_Additive
    translationMode=BMM_Additive
    alphaBoolBlend {blendOption=Linear}
    boneToModify {boneName=dyn_hair_r_02}
    rotation {pitch=-10}
    translation {z=0.5}
   }
   n7 = "Transform (Modify) Bone - Bone: dyn_hair_bk_r_03"(ComponentPose=@n6, Alpha=@n5, Translation="0,0,1.5", Rotation="-15,0,-20", Scale="1,1,1") {
    rotationMode=BMM_Additive
    scaleMode=BMM_Additive
    translationMode=BMM_Additive
    alphaBoolBlend {blendOption=Linear}
    boneToModify {boneName=dyn_hair_bk_r_03}
    rotation {pitch=-15, roll=-20}
    translation {z=1.5}
   }
   n8 = Get Rifle_Adjustment()
   n9 = "Transform (Modify) Bone - Bone: dyn_hair_bk_02"(ComponentPose=@n7, Alpha=@n8, Translation="0,0,0.75", Rotation="0,0,20", Scale="1,1,1") {
    rotationMode=BMM_Additive
    scaleMode=BMM_Additive
    translationMode=BMM_Additive
    alphaBoolBlend {blendOption=Linear}
    boneToModify {boneName=dyn_hair_bk_02}
    rotation {roll=20}
    translation {z=0.75}
   }
   n10 = "Transform (Modify) Bone - Bone: dyn_hair_bk_r_02"(ComponentPose=@n9, Alpha=@n8, Translation="0,0,0", Rotation="-12,0,10", Scale="1,1,1") {
    rotationMode=BMM_Additive
    scaleMode=BMM_Additive
    translationMode=BMM_Additive
    alphaBoolBlend {blendOption=Linear}
    boneToModify {boneName=dyn_hair_bk_r_02}
    rotation {pitch=-12, roll=10}
   }
   Blend Poses by bool(BlendPose_0=Component To Local(ComponentPose=@n10), bActiveValue=OR Boolean(A=Get bIsFalling(), B=Get bIsJumping())) {blendType=Linear}
   n11 = Get Aim_Up()
   n12 = "Transform (Modify) Bone - Bone: dyn_hair_r_02"(Alpha=@n11, Translation="0,0,0", Rotation="-30,20,15", Scale="1,1,1") {
    rotationMode=BMM_Additive
    scaleMode=BMM_Additive
    translationMode=BMM_Additive
    alphaBoolBlend {blendOption=Linear}
    boneToModify {boneName=dyn_hair_r_02}
    rotation=<rot -30 20 15>
   }
   n13 = "Transform (Modify) Bone - Bone: dyn_hair_bk_r_03"(ComponentPose=@n12, Alpha=@n11, Translation="0,0,0.75", Rotation="-30,0,-20", Scale="1,1,1") {
    rotationMode=BMM_Additive
    scaleMode=BMM_Additive
    translationMode=BMM_Additive
    alphaBoolBlend {blendOption=Linear}
    boneToModify {boneName=dyn_hair_bk_r_03}
    rotation {pitch=-30, roll=-20}
    translation {z=0.75}
   }
   n14 = Get Rifle_Adjustment()
   n15 = "Transform (Modify) Bone - Bone: dyn_hair_r_02"(ComponentPose=@n13, Alpha=@n14, Translation="0,0,0.5", Rotation="-10,0,0", Scale="1,1,1") {
    rotationMode=BMM_Additive
    scaleMode=BMM_Additive
    translationMode=BMM_Additive
    alphaBoolBlend {blendOption=Linear}
    boneToModify {boneName=dyn_hair_r_02}
    rotation {pitch=-10}
    translation {z=0.5}
   }
   n16 = "Transform (Modify) Bone - Bone: dyn_hair_bk_r_03"(ComponentPose=@n15, Alpha=@n14, Translation="0,0,1.5", Rotation="-15,0,-20", Scale="1,1,1") {
    rotationMode=BMM_Additive
    scaleMode=BMM_Additive
    translationMode=BMM_Additive
    alphaBoolBlend {blendOption=Linear}
    boneToModify {boneName=dyn_hair_bk_r_03}
    rotation {pitch=-15, roll=-20}
    translation {z=1.5}
   }
   n17 = Get Rifle_Adjustment()
   n18 = "Transform (Modify) Bone - Bone: dyn_hair_bk_02"(ComponentPose=@n16, Alpha=@n17, Translation="0,0,0.75", Rotation="0,0,20", Scale="1,1,1") {
    rotationMode=BMM_Additive
    scaleMode=BMM_Additive
    translationMode=BMM_Additive
    alphaBoolBlend {blendOption=Linear}
    boneToModify {boneName=dyn_hair_bk_02}
    rotation {roll=20}
    translation {z=0.75}
   }
   "Transform (Modify) Bone - Bone: dyn_hair_bk_r_02"(ComponentPose=@n18, Alpha=@n17, Translation="0,0,0", Rotation="-12,0,10", Scale="1,1,1") {
    rotationMode=BMM_Additive
    scaleMode=BMM_Additive
    translationMode=BMM_Additive
    alphaBoolBlend {blendOption=Linear}
    boneToModify {boneName=dyn_hair_bk_r_02}
    rotation {pitch=-12, roll=10}
   }
   n19 = RigidBody(ComponentPose=Local To Component(LocalPose=Copy Pose From Mesh() {bUseAttachedParent=true})) {
    bClampLinearTranslationLimitToRefPose=true
    bOverrideWorldGravity=true
    overridePhysicsAsset=&/M_MED_Wombat_FaceAcc_Poncho_Dyn
    simulationSpace=BaseBoneSpace
    alphaBoolBlend {blendOption=Linear}
    baseBoneRef {boneName=neck_02}
    componentAppliedLinearAccClamp=<3000 3000 3000>
    componentLinearAccScale=<0.45 0.25 0.23>
    componentLinearVelScale=<0.35 0.2 0.2>
    overrideWorldGravity {z=-980}
   }
   n20 = "Constraint - dyn_hair_l_01 (P)"(ComponentPose=@n19, Alpha=1, ConstraintWeights_0=1) {
    alphaBoolBlend {blendOption=Linear}
    boneToModify {boneName=dyn_hair_l_01}
    constraintSetup [{targetBone {boneName=head}}]
    constraintWeights [1]
   }
   n21 = "Constraint - dyn_hair_bk_l_01 (P)"(ComponentPose=@n20, Alpha=1, ConstraintWeights_0=1) {
    alphaBoolBlend {blendOption=Linear}
    boneToModify {boneName=dyn_hair_bk_l_01}
    constraintSetup [{targetBone {boneName=head}}]
    constraintWeights [1]
   }
   n22 = "Constraint - dyn_hair_bk_01 (P)"(ComponentPose=@n21, Alpha=1, ConstraintWeights_0=1) {
    alphaBoolBlend {blendOption=Linear}
    boneToModify {boneName=dyn_hair_bk_01}
    constraintSetup [{targetBone {boneName=head}}]
    constraintWeights [1]
   }
   n23 = "Constraint - dyn_hair_bk_r_03 (P)"(ComponentPose=@n22, Alpha=1, ConstraintWeights_0=1) {
    alphaBoolBlend {blendOption=Linear}
    boneToModify {boneName=dyn_hair_bk_r_03}
    constraintSetup [{targetBone {boneName=head}}]
    constraintWeights [1]
   }
   n24 = "Constraint - dyn_hair_r_01 (P)"(ComponentPose=@n23, Alpha=1, ConstraintWeights_0=1) {
    alphaBoolBlend {blendOption=Linear}
    boneToModify {boneName=dyn_hair_r_01}
    constraintSetup [{targetBone {boneName=head}}]
    constraintWeights [1]
   }
   n25 = "Constraint - dyn_hair_l_02 (P)"(ComponentPose=@n24, Alpha=0.6, ConstraintWeights_0=1) {
    alphaBoolBlend {blendOption=Linear}
    boneToModify {boneName=dyn_hair_l_02}
    constraintSetup [{targetBone {boneName=head}}]
    constraintWeights [1]
   }
   n26 = "Constraint - dyn_hair_bk_l_02 (P)"(ComponentPose=@n25, Alpha=0.8, ConstraintWeights_0=1) {
    alphaBoolBlend {blendOption=Linear}
    boneToModify {boneName=dyn_hair_bk_l_02}
    constraintSetup [{targetBone {boneName=head}}]
    constraintWeights [1]
   }
   n27 = "Constraint - dyn_hair_bk_02 (P)"(ComponentPose=@n26, Alpha=0, ConstraintWeights_0=1) {
    alphaBoolBlend {blendOption=Linear}
    boneToModify {boneName=dyn_hair_bk_02}
    constraintSetup [{targetBone {boneName=neck_01}}]
    constraintWeights [1]
   }
   n28 = "Constraint - dyn_hair_bk_r_03 (P)"(ComponentPose=@n27, Alpha=0, ConstraintWeights_0=1) {
    alphaBoolBlend {blendOption=Linear}
    boneToModify {boneName=dyn_hair_bk_r_03}
    constraintSetup [{targetBone {boneName=neck_01}}]
    constraintWeights [1]
   }
   n29 = "Constraint - dyn_hair_r_02 (P)"(ComponentPose=@n28, Alpha=0, ConstraintWeights_0=1) {
    alphaBoolBlend {blendOption=Linear}
    boneToModify {boneName=dyn_hair_r_02}
    constraintSetup [{targetBone {boneName=head}}]
    constraintWeights [1]
   }
   n30 = "Constraint - dyn_hair_l_03 (P)"(ComponentPose=@n29, Alpha=0.3, ConstraintWeights_0=1) {
    alphaBoolBlend {blendOption=Linear}
    boneToModify {boneName=dyn_hair_l_03}
    constraintSetup [{targetBone {boneName=neck_02}}]
    constraintWeights [1]
   }
   n31 = "Constraint - dyn_hair_bk_l_03 (P)"(ComponentPose=@n30, Alpha=1, ConstraintWeights_0=1) {
    alphaBoolBlend {blendOption=Linear}
    boneToModify {boneName=dyn_hair_bk_l_03}
    constraintSetup [{targetBone {boneName=neck_02}}]
    constraintWeights [1]
   }
   n32 = "Constraint - dyn_hair_bk_03 (P)"(ComponentPose=@n31, Alpha=1, ConstraintWeights_0=1) {
    alphaBoolBlend {blendOption=Linear}
    boneToModify {boneName=dyn_hair_bk_03}
    constraintSetup [{targetBone {boneName=neck_02}}]
    constraintWeights [1]
   }
   n33 = "Constraint - dyn_hair_bk_r_04 (P)"(ComponentPose=@n32, Alpha=1, ConstraintWeights_0=1) {
    alphaBoolBlend {blendOption=Linear}
    boneToModify {boneName=dyn_hair_bk_r_04}
    constraintSetup [{targetBone {boneName=neck_02}}]
    constraintWeights [1]
   }
   n34 = "Constraint - dyn_hair_r_03 (P)"(ComponentPose=@n33, Alpha=0, ConstraintWeights_0=1) {
    alphaBoolBlend {blendOption=Linear}
    boneToModify {boneName=dyn_hair_r_03}
    constraintSetup [{targetBone {boneName=neck_02}}]
    constraintWeights [1]
   }
   n35 = "Constraint - dyn_hair_bk_r_01 (P)"(ComponentPose=@n34, Alpha=0.3, ConstraintWeights_0=1) {
    alphaBoolBlend {blendOption=Linear}
    boneToModify {boneName=dyn_hair_bk_r_01}
    constraintSetup [{targetBone {boneName=neck_01}}]
    constraintWeights [1]
   }
   Output Pose(Result=Component To Local(ComponentPose=@n35)) {layerGroup=None}
  ]}
 EventGraph {nodes [
   Event BlueprintUpdateAnimation()
   n1 = BreakTransform(InTransform=GetDeltaTransformFromRefPose(self=GetOwnerSkeletalMeshComponent(), BoneName=upperarm_r, BaseName=clavicle_r))
   > Set Aim_Up(Aim_Up:double=MapRangeClamped(Value=BreakRotator(InRot=@n1.Rotation).Yaw, InRangeA=90, InRangeB=120)) -> (Output_Get:double)
   n2 = BreakTransform(InTransform=GetDeltaTransformFromRefPose(self=GetOwnerSkeletalMeshComponent(), BoneName=clavicle_r, BaseName=spine_05))
   > Set Rifle_Adjustment(Rifle_Adjustment:double=MapRangeClamped(Value=BreakRotator(InRot=@n2.Rotation), InRangeA=-2, InRangeB=-4)) -> (Output_Get:double)
   n3 = GetOwnerSkeletalMeshComponent()
   n4 = BreakRotator(InRot=BreakTransform(InTransform=GetDeltaTransformFromRefPose(self=@n3, BoneName=head, BaseName=neck_02)).Rotation)
   float - float(A=MapRangeClamped(Value=@n4.Yaw, InRangeA=0, InRangeB=-12), B=MapRangeClamped(Value=@n4.Yaw, InRangeA=-13, InRangeB=-20)) {is=title K2Node_CallFunction}
  ]}
]
variables [
 Aim_Up {type=number}
 Head_Adjust {type=number}
 Rifle_Adjustment {type=number}
]
